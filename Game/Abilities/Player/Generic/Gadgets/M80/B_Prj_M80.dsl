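; M80 gadget projectile that loops a fuse sound and explodes on a timer or on impact.
; BeginPlay starts the fuse sound and waits a random delay between the setup min and max.
; Hits on a FortPawn or on actors with matching explode-on-hit tags detonate it early, with DoOnce gating DoExplosion.
; OnExploded spawns the particle effect, force feedback and explosion sound at its location, then fades the fuse audio.
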(event EventBeginPlay
  (Audio|Components|Audio|Play (Variables|Projectile_Throwing_Stars|GetCue_Fuse_Loop_Sound) 0.0)
  (Utilities|FlowControl|Delay (Math|Random|RandomFloatinRange (Variables|Setup|GetExplo_Delay_Min) (Variables|Setup|GetExplo_Delay_Max))) ; Time Before Explosion
  (Utilities|FlowControl|DoOnce false
    (:Completed
      (Audio|Components|Audio|Stop (Variables|Projectile_Throwing_Stars|GetCue_Fuse_Loop_Sound))
      (Delivery|DoExplosion)
      (Audio|Components|Audio|FadeOut (Variables|Projectile_Throwing_Stars|GetCue_Fuse_Loop_Sound) 0.2 0.0 "Linear" :!class /Script/Engine.AudioComponent))))

(event Collision|EventHit
  (params (MyComp :type Primitive%20Component%20Object%20Reference) (Other :type Actor%20Object%20Reference) (OtherComp :type Primitive%20Component%20Object%20Reference) (bSelfMoved :type Boolean) (HitLocation :type Vector :default "0, 0, 0") (HitNormal :type Vector :default "0, 0, 0") (NormalImpulse :type Vector :default "0, 0, 0") (Hit :type Hit%20Result%20Structure%20%28by%20ref%29))
  (Variables|Default|SetHitLocation HitLocation)
  (Utilities|Casting|CastToFortPawn Other
    (:then
      _)
    (:CastFailed
      (bind _returnvalue (GameplayTags|HasAnyMatchingGameplayTags(Message) Other (Variables|Default|GetTC_ActorTagsThatShouldExplodeOnHit)))
      (if _returnvalue
        _)))
  (Utilities|FlowControl|DoOnce false
    (:Completed
      (Audio|Components|Audio|Stop (Variables|Projectile_Throwing_Stars|GetCue_Fuse_Loop_Sound))
      (Delivery|DoExplosion)
      (Audio|Components|Audio|FadeOut (Variables|Projectile_Throwing_Stars|GetCue_Fuse_Loop_Sound) 0.2 0.0 "Linear" :!class /Script/Engine.AudioComponent))))

(event EventOnExploded
  (params (HitActors :type Array%20of%20Actor%20Object%20References) (HitResults :type Array%20of%20Hit%20Result%20Structures))
  (bind _returnvalue (Transformation|GetActorLocation))
  (Effects|SpawnEmitteratLocation (Variables|Default|GetP_ImpactEffects) _returnvalue)
  (Fortnite|PlayLocalForceFeedbackatLocation (Variables|Default|GetExplosionForceFeedback) _returnvalue 200)
  (Audio|PlaySoundatLocation (Variables|Default|GetCue_ExplosionSound) _returnvalue)
  (Audio|Components|Audio|FadeOut (Variables|Projectile_Throwing_Stars|GetCue_Fuse_Loop_Sound) 0.2 0.0 "Linear" :!class /Script/Engine.AudioComponent))

(fn ConstructionScript)
